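# Physics asset for the Flamingo 2 two-handed pickaxe, defining simulation bodies for its neck, feathers, flipper and rope.
# Capsules cover most bones, with boxes on the left, middle and right feathers.
# Feathers and neck hang off dyn_base with tight swing limits, feather twist being locked.
# The flipper swings mainly on one axis from the handle, which also anchors a two-link rope chain; no constraint joins dyn_base to the handle.
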PhysicsAsset /BRCosmetics/Weapons/FORT_Melee/Pickaxe_Flamingo2/Meshes/Pickaxe_Flamingo_2_Physics
bodies {
 Melee_TwoHanded_Handle [{shapeType=Capsule, length=85.9063, radius=2.6343, center=<0.588 -0.6474 9.1147>}]
 dyn_base [
  {shapeType=Capsule, length=7.7412, radius=10.3021, center=<15.1971 -7.9751 0.0002>, rotation=<rot -90 -4.7636 -105.2357>}
  {shapeType=Capsule, length=102.5, radius=31.25, center=<-18.529 -4.6296 -0.0003>, rotation=<rot 90 -2.5305 -2.5304>}
 ]
 dyn_flipper [{shapeType=Capsule, length=9.403, radius=5.0456, center=<7.8387 -1.1452 0>, rotation=<rot 90 0 10.0001>}]
 dyn_l_feather_02 [{shapeType=Box, extentX=16.7031, extentY=1.7055, extentZ=7.0778, center=<2.9573 -0.7043 -3.7809>, rotation=<rot -10 0 0>}]
 dyn_mid_feather_02 [{shapeType=Box, extentX=12.0106, extentY=1.2658, extentZ=3.4346, center=<3.022 0.3724 0>}]
 dyn_neck_01 [{shapeType=Capsule, length=4.3251, radius=3.762, center=<-0.9207 -0.6564 -0.0003>, rotation=<rot 0.0001 125.0004 89.9998>}]
 dyn_r_feather_02 [{shapeType=Box, extentX=14.8411, extentY=1.7962, extentZ=5.9195, center=<4.8762 -1.2972 2.506>, rotation=<rot 10 0 0>}]
 dyn_rope_01 [{shapeType=Capsule, length=2.2967, radius=1.2242, center=<1.4523 -0.2485 -0.2354>, rotation=<rot 0 -114.481 -83.2387>}]
 dyn_rope_02 [{shapeType=Capsule, length=2.356, radius=1.237, center=<1.1734 -0.0593 0.0206>, rotation=<rot 0 85.5242 -85.2244>}]
 dyn_top_feather_01 [{shapeType=Capsule, length=6.3048, radius=2.2912, center=<1.979 0 0>, rotation=<rot -90 -1.2659 1.2658>}]
 dyn_top_feather_02 [{shapeType=Capsule, length=6.3048, radius=2.2912, center=<2.0882 0 0>, rotation=<rot -90 -179.9802 179.9803>}]
 dyn_top_feather_03 [{shapeType=Capsule, length=6.3048, radius=2.2912, center=<2.7516 0.0002 0>, rotation=<rot -90 0.0033 -0.0032>}]
 dyn_top_feather_04 [{shapeType=Capsule, length=6.4385, radius=2.1574, center=<2.6658 -0.1233 0>, rotation=<rot -90 2.5305 -2.5304>}]
 dyn_top_feather_05 [{shapeType=Capsule, length=6.3048, radius=2.2912, rotation=<rot -90 5.0512 -5.0511>}]
}
constraints [
 {bone1Name=dyn_top_feather_01, bone2Name=dyn_base, swing1LimitDegrees=10, swing2LimitDegrees=10, twistLimitDegrees=45, twistMotion=Locked}
 {bone1Name=dyn_top_feather_02, bone2Name=dyn_base, swing1LimitDegrees=10, swing2LimitDegrees=10, twistLimitDegrees=45, twistMotion=Locked}
 {bone1Name=dyn_top_feather_03, bone2Name=dyn_base, swing1LimitDegrees=10, swing2LimitDegrees=10, twistLimitDegrees=45, twistMotion=Locked}
 {bone1Name=dyn_top_feather_04, bone2Name=dyn_base, swing1LimitDegrees=10, swing2LimitDegrees=10, twistLimitDegrees=45, twistMotion=Locked}
 {bone1Name=dyn_top_feather_05, bone2Name=dyn_base, swing1LimitDegrees=10, swing2LimitDegrees=10, twistLimitDegrees=45, twistMotion=Locked}
 {bone1Name=dyn_r_feather_02, bone2Name=dyn_base, swing1LimitDegrees=5, swing2LimitDegrees=5, twistLimitDegrees=45, twistMotion=Locked}
 {bone1Name=dyn_l_feather_02, bone2Name=dyn_base, swing1LimitDegrees=5, swing2LimitDegrees=5, twistLimitDegrees=45, twistMotion=Locked}
 {bone1Name=dyn_mid_feather_02, bone2Name=dyn_base, swing1LimitDegrees=5, swing2LimitDegrees=5, twistLimitDegrees=45, twistMotion=Locked}
 {bone1Name=dyn_neck_01, bone2Name=dyn_base, swing1LimitDegrees=6, swing2LimitDegrees=1, twistLimitDegrees=1}
 {bone1Name=dyn_flipper, bone2Name=Melee_TwoHanded_Handle, swing1LimitDegrees=45, swing2LimitDegrees=0.1, twistLimitDegrees=0.1}
 {bone1Name=dyn_rope_01, bone2Name=Melee_TwoHanded_Handle, swing1LimitDegrees=15, swing2LimitDegrees=15, twistLimitDegrees=5}
 {bone1Name=dyn_rope_02, bone2Name=dyn_rope_01, swing1LimitDegrees=15, swing2LimitDegrees=15, twistLimitDegrees=5}
]
